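# UI material function for a reticle's radial ammo bar that doubles as a reload timer.
# Active ammo divided by clip size fills an arc with one segment per round, drawn by MF_UI_Bar_Segmented_Radial_Progress.
# Any Reload Progress above zero flips a branchless Lerp switch to show reload progress as a near-continuous arc, using a very high segment count and no spacing.
# It outputs opacity and progress masks.
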MaterialFunction /FortUILibrary/Systems/Reticles/MaterialFunctions/AuxillaryProgressBars/MF_UI_Reticle_Bar_Radial_Ammo
graph {
 properties {
  bExposeToLibrary=true
  description=Creates Ammo Bar with Reload Timer Functionality..
  previewBlendMode=BLEND_Translucent
  previewMaterialDomain=MD_UI
 }
 nodes [
  n1 = FunctionInput() {
   bUsePreviewValueAsDefault=true
   description=Defines artc start angle of meter
   inputName="Start Angle (0-1)"
   inputType=FunctionInput_Scalar
   sortPriority=4
   previewValue=<0.25 0 0 1>
  }
  n2 = FunctionInput() {
   bUsePreviewValueAsDefault=true
   description=Define arc end angle or meter
   inputName="End Angle (0-1)"
   inputType=FunctionInput_Scalar
   sortPriority=5
   previewValue=<0.5 0 0 1>
  }
  n3 = FunctionInput() {
   bUsePreviewValueAsDefault=true
   description=The opacity of the backing element for the meter
   inputName="Background Opacity (0-1)"
   inputType=FunctionInput_Scalar
   sortPriority=7
   previewValue=<0.5 0 0 1>
  }
  n4 = FunctionInput() {
   bUsePreviewValueAsDefault=true
   description=Defines the size of the circle mask where 1 is the edge of the border.
   inputName="Circle Size (0-1)"
   inputType=FunctionInput_Scalar
   sortPriority=8
   previewValue=<0.5 0 0 1>
  }
  n5 = FunctionInput() {
   bUsePreviewValueAsDefault=true
   description=Defubes the stroke thickness of the circle mask.
   inputName=Outline Thickness
   inputType=FunctionInput_Scalar
   sortPriority=9
   previewValue=<0.025 0 0 1>
  }
  n6 = FunctionInput() {
   bUsePreviewValueAsDefault=true
   description=Define if the spacing is fixed
   inputName=Fixed Spacing
   inputType=FunctionInput_Scalar
   sortPriority=10
  }
  n7 = FunctionInput() {
   bUsePreviewValueAsDefault=true
   description=The Meter Progress of the Reload Timer
   inputName=Reload Progress
   inputType=FunctionInput_Scalar
   sortPriority=1
  }
  n8 = Ceil(None=MaterialFunctionCall!(A=@n7, B=Constant()) {materialFunction=/FortUILibrary/MaterialFunctions/Basics/MF_UI_BooleanOperations}.IsGreater) {desc=Controls Switches}
  n9 = FunctionInput() {
   bUsePreviewValueAsDefault=true
   description="Used for the current usable ammo. Gets divided by the Segment Count (Clip Size)"
   inputName=Active Ammo
   inputType=FunctionInput_Scalar
   previewValue=<10 0 0 1>
  }
  n10 = FunctionInput() {
   bUsePreviewValueAsDefault=true
   description=Used for defininh the total ammount of ammo of the ability/weapon.
   inputName=Clip Size
   inputType=FunctionInput_Scalar
   previewValue=<20 0 0 1>
  }
  n11 = FunctionInput() {
   bUsePreviewValueAsDefault=true
   description=Defines spacing between sectors
   inputName=Segment Spacing
   inputType=FunctionInput_Scalar
   sortPriority=2
   previewValue=<0.2 0 0 1>
  }
  n12 = MaterialFunctionCall!(
   UVs=FunctionInput(Preview=MaterialFunctionCall() {materialFunction=&GetUserInterfaceUV}.Normalized UV) {bUsePreviewValueAsDefault=true, inputName=UVs, inputType=FunctionInput_Vector2, sortPriority=-1}
   Meter Progress=LinearInterpolate(A=Divide(A=@n9, B=@n10), B=@n7, Alpha=@n8) {constA=0.1, constB=0, desc=Switch for Meter Progress}
   Delta Progress
   Segment Count=LinearInterpolate(A=@n10, B=Constant() {desc=Reload Segment Steps, r=1000}, Alpha=@n8) {desc=Switch for Segment Count}
   Segment Spacing=LinearInterpolate(A=@n11, Alpha=@n8) {constA=0.2, constB=0, desc=Switch for Segment Spacing}
   "Start Angle (0-1)"=@n1
   "End Angle (0-1)"=@n2
   Fill Direction
   "Background Opacity (0-1)"=@n3
   "Circle Size (0-1)"=@n4
   Outline Thickness=@n5
   Fixed Spacing=@n6
   Snap Progress to Segments
  ) {materialFunction=&/MF_UI_Bar_Segmented_Radial_Progress}
  FunctionOutput(None=@n12.Opacity Mask) {outputName=Opacity Mask, sortPriority=-1}
  FunctionOutput(None=@n12.Meter Progress Mask) {description="The combined fill+backing meter. ", outputName=Progress Mask}
 ]
}
